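# Translucent UI material for the BeanDog weapon reticle, with brackets, circles, a center dot and two lines.
# Brackets come from an SDF texture with mirrored UVs, while UI shape functions draw the rest procedurally, including outlines and an SDF glow.
# Named reroutes carry the masks into color and opacity blends, where IsInRange switches appearance and scalar parameters control layer opacities.
# A distortion function also offsets a shared UV set.
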Material /WildEstateWeapons/UI/BeanDog/Materials/M_UI_BeanDog_Reticle
graph {
 properties {bAutomaticallySetUsageInEditor=false, blendMode=BLEND_Translucent, materialDomain=MD_UI, usedShadingModels=Default Lit}
 nodes [
  n1 = ScalarParameter() {parameterName=Brackets Opacity}
  n2 = Add(A=NamedRerouteUsage(), B=NamedRerouteUsage())
  n3 = NamedRerouteUsage()
  n4 = LinearInterpolate(A=NamedRerouteUsage(), B=@n3, Alpha=@n3)
  n5 = ScalarParameter() {parameterName=LinesOpacity}
  n6 = NamedRerouteUsage()
  n7 = NamedRerouteUsage()
  n8 = Add(A=Add(A=Multiply(A=@n1, B=@n2), B=Multiply(A=@n4, B=@n5)), B=Multiply(A=LinearInterpolate(A=@n6, B=@n7, Alpha=@n7), B=@n5))
  n9 = NamedRerouteUsage()
  n10 = NamedRerouteUsage()
  Add(A=@n8, B=Multiply(A=LinearInterpolate(A=@n9, B=@n10, Alpha=@n10), B=ScalarParameter() {defaultValue=1, parameterName=Center Dot opacity}))
  Constant3Vector() {constant=<rgba 0.3891 1 0 1>}
  n11 = NamedRerouteUsage()
  LinearInterpolate(A=Multiply(B=NamedRerouteUsage()), B=@n11, Alpha=@n11)
  n12 = Add(A=Add(A=NamedRerouteUsage(), B=NamedRerouteUsage()), B=NamedRerouteUsage())
  n13 = LinearInterpolate(A=NamedRerouteUsage(), B=NamedRerouteUsage(), Alpha=ScalarParameter() {parameterName=IsInRange})
  LinearInterpolate(
   A=LinearInterpolate(Alpha=Add(A=@n12, B=@n13)) {constA=1, constB=0}
   B=Constant3Vector() {constant=<rgba 0.3891 1 0 1>}
   Alpha=Subtract(A=NamedRerouteUsage(), B=@n13)
  )
  n14 = NamedRerouteUsage()
  n15 = NamedRerouteUsage()
  n16 = ScalarParameter() {parameterName=LinesOpacity}
  n17 = Add(A=@n11, B=Multiply(A=LinearInterpolate(A=@n14, B=@n15, Alpha=@n15), B=@n16))
  n18 = NamedRerouteUsage()
  n19 = Multiply(A=LinearInterpolate(A=NamedRerouteUsage(), B=@n18, Alpha=@n18), B=@n16)
  n20 = NamedRerouteUsage()
  n21 = NamedRerouteUsage()
  n22 = LinearInterpolate(
   A=LinearInterpolate(A=Multiply(A=@n20) {constB=0.6}, B=NamedRerouteUsage(), Alpha=ScalarParameter() {parameterName=IsInRange})
   B=@n21
   Alpha=@n21
  )
  n23 = ScalarParameter() {defaultValue=1, parameterName=CircleOpacity}
  Multiply(A=Add(A=Add(A=@n17, B=@n19), B=Multiply(A=@n22, B=@n23)), B=ScalarParameter() {defaultValue=1, parameterName=GlobalOpacity})
  n24 = Constant3Vector() {constant=<rgba 1 1 1 0>}
  n25 = Add(A=LinearInterpolate(A=Constant3Vector(), B=@n24, Alpha=NamedRerouteUsage()), B=Multiply(A=@n24, B=NamedRerouteUsage()))
  n26 = Multiply(A=@n24, B=NamedRerouteUsage())
  n27 = NamedRerouteUsage()
  NamedRerouteDeclaration(None=Add(A=Add(A=@n25, B=@n26), B=Multiply(A=@n24, B=@n27))) {name=Final Color, nodeColor=<rgba 0 0.9297 1 1>}
  n28 = NamedRerouteUsage()
  n29 = NamedRerouteUsage()
  NamedRerouteDeclaration(None=Add(A=@n8, B=LinearInterpolate(A=@n28, B=@n29, Alpha=@n29))) {name=Opacity, nodeColor=<rgba 1 0 0.0937 1>}
  n30 = Constant()
  n31 = Constant() {r=-0.11}
  n32 = MaterialFunctionCall(UVs=Add(A=TextureCoordinate(), B=AppendVector(A=@n30, B=@n31)) {constB=0}, Size=Constant() {r=0.04}, OutlineThickness=Constant() {r=0.02}) {materialFunction=&MF_UI_Circle}
  n33 = NamedRerouteDeclaration(None=@n32.Outline) {name=CircleOutline, nodeColor=<rgba 0 0 0 1>}
  n34 = ScalarParameter() {parameterName=IsInRange}
  NamedRerouteDeclaration(None=Max(A=NamedRerouteUsage(), B=LinearInterpolate(A=@n32.Fill, B=@n33, Alpha=@n34))) {name=Circle, nodeColor=<rgba 0 0 0 1>}
  n35 = Constant()
  n36 = Constant() {r=-0.15}
  n37 = MaterialFunctionCall(UVs=Add(A=TextureCoordinate(), B=AppendVector(A=@n35, B=@n36)) {constB=0}, Size=Constant() {r=0.025}, OutlineThickness=Constant() {r=0.02}) {materialFunction=&MF_UI_Circle}
  NamedRerouteDeclaration(None=@n37.Outline) {name=Center dot outline, nodeColor=<rgba 0.9219 0 1 1>}
  n38 = TextureCoordinate()
  n39 = AppendVector(A=Constant(), B=Constant())
  NamedRerouteDeclaration(None=MaterialFunctionCall(UVs=Add(A=@n38, B=@n39) {constB=0}, Size=Constant() {r=0.1}, OutlineThickness=Constant() {r=0.02}) {materialFunction=&MF_UI_Circle}.Outline) {name=Center Circle, nodeColor=<rgba 0 0.2266 1 1>}
  n40 = Constant() {r=0.2}
  NamedRerouteDeclaration(None=Multiply(
    A=MaterialFunctionCall!(DistanceField=@n32.SDF, GlowMin, GlowMax=@n40) {materialFunction=/FortUILibrary/MaterialFunctions/Shapes/Modifiers/MF_UI_SDFGlowOrAA}.Result
    B=ScalarParameter() {defaultValue=0.4, parameterName=GlowIntensity}
   ) {constB=0.7}) {name=Glow, nodeColor=<rgba 0 0 0 1>}
  n41 = TextureCoordinate()
  n42 = AppendVector(A=Constant(), B=Constant())
  NamedRerouteDeclaration(None=MaterialFunctionCall(UVs=Add(A=@n41, B=@n42) {constB=0}, Size=Constant() {r=0.1}, OutlineThickness=Constant() {r=0.05}) {materialFunction=&MF_UI_Circle}.Outline) {name=Center Circle Outline, nodeColor=<rgba 0.5469 1 0 1>}
  n43 = TextureCoordinate()
  n44 = AppendVector(A=Constant(), B=Constant())
  NamedRerouteDeclaration(None=MaterialFunctionCall(UVs=Add(A=@n43, B=@n44) {constB=0}, Size=Constant() {r=0.07}, OutlineThickness=Constant() {r=0.01}, GlowMax=Constant() {r=0.08}) {materialFunction=&MF_UI_Circle}.Outline) {name=Center Circle Glow, nodeColor=<rgba 1 0.7734 0 1>}
  n45 = MaterialFunctionCall(Texture Scale=Constant2Vector() {g=0.25, r=0.6}) {materialFunction=&ScaleUVsByCenter}
  n46 = Constant() {r=-0.3}
  n47 = Constant() {r=-0.5}
  n48 = TextureSample(UVs=Add(
    A=MaterialFunctionCall!(UV=@n45.UVs, Mirror Y, Swizzle UV) {materialFunction=/FortUILibrary/MaterialFunctions/Basics/MF_UI_Symmetry_2x1_New}.Flipped Symmetry
    B=AppendVector(A=@n46, B=@n47)
   ) {constB=0}) {samplerSource=SSM_Clamp_WorldGroupSettings, texture=/WildEstateWeapons/UI/BeanDog/Textures/T_UI_BeanDog_Reticle_Bracket_SDF}
  NamedRerouteDeclaration(None=MaterialFunctionCall(DistanceField=@n48.RGB, Threshold=Constant() {r=0.6}) {materialFunction=&MF_UI_AntiAliasing}.Outer) {name=Reticle, nodeColor=<rgba 0 0 0 1>}
  NamedRerouteDeclaration(None=MaterialFunctionCall(Size=Constant() {r=0.1}, OutlineThickness=Constant() {r=0.01}) {materialFunction=&MF_UI_Circle}.Outline) {name=OutlineInRange, nodeColor=<rgba 0 0 0 1>}
  NamedRerouteDeclaration(None=@n37.Fill) {name=Center Dot, nodeColor=<rgba 0.7578 1 0 1>}
  n49 = Constant() {r=0.02}
  n50 = Constant() {r=0.02}
  n51 = MaterialFunctionCall!(
   UVs
   PointA=Constant2Vector() {g=0.55, r=-0.25}
   PointB=Constant2Vector() {g=0.55, r=0.25}
   Thickness=@n49
   OutlineThickness=@n50
   OutlineAlignment
   GlowMin
   GlowMax
  ) {materialFunction=/FortUILibrary/MaterialFunctions/Shapes/MF_UI_LineSegment}
  NamedRerouteDeclaration(None=@n51.Fill) {name=Line1, nodeColor=<rgba 0 0 0 1>}
  n52 = MaterialFunctionCall(PointA=Constant2Vector() {g=0.7, r=-0.1}, PointB=Constant2Vector() {g=0.7, r=0.1}, Thickness=@n49, OutlineThickness=@n50) {materialFunction=&MF_UI_LineSegment}
  NamedRerouteDeclaration(None=@n52.Fill) {name=Line2, nodeColor=<rgba 0 0 0 1>}
  n53 = Constant()
  n54 = Constant() {r=0.02}
  NamedRerouteDeclaration(None=Add(
    A=TextureCoordinate()
    B=MaterialFunctionCall!(UV, SpeedX=@n53, SpeedY=@n53, Intensity=@n54) {materialFunction=/WildEstateWeapons/UI/MaterialFunctions/MF_UI_WildEstate_Distort}.Result
   )) {name=UV's, nodeColor=<rgba 0 0 0 1>}
  NamedRerouteDeclaration(None=MaterialFunctionCall(DistanceField=@n48.RGB, Threshold=Constant() {r=0.4}) {materialFunction=&MF_UI_AntiAliasing}.Outer) {name=Outline, nodeColor=<rgba 0 0 0 1>}
  NamedRerouteDeclaration(None=@n51.Outline) {name=Line1Outline, nodeColor=<rgba 0 0 0 1>}
  NamedRerouteDeclaration(None=@n52.Outline) {name=Line2Outline, nodeColor=<rgba 0 0 0 1>}
  NamedRerouteUsage()
  NamedRerouteUsage()
  NamedRerouteUsage()
  NamedRerouteUsage()
  NamedRerouteUsage()
  NamedRerouteUsage()
  NamedRerouteUsage()
  Power(Base=@n32.Outline) {constExponent=0.01}
 ]
}
