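# Weapon reticle extension widget for the BeanDog bow, showing an Image that uses the M_UI_BeanDog_Reticle material.
# Construct casts the equipped weapon to B_BeanDog and caches it, printing a message if the cast fails.
# Tick checks that the bow's trajectory is valid, casts it to the explosive bow trajectory and prints its spline length every frame.
# The graph only produces this debug output and never updates the reticle image or its material.
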WidgetBlueprint /WildEstateWeapons/UI/BeanDog/WBP_BeanDogReticle
graphs [EventGraph {nodes [
   Event Construct()
   > n1 = Cast To B_BeanDog!(
    execute
    Object=GetEquippedWeapon!(self:Fort Weapon Reticle Extension Widget Base Object Reference) -> (ReturnValue:Fort Weapon Object Reference)
   ) -> (then, CastFailed, AsB Bean Dog:B Bean Dog Object Reference) {is=title K2Node_DynamicCast}
   > Set As B Bean Dog(As B Bean Dog:B Bean Dog Object Reference=@n1) -> (Output_Get:B Bean Dog Object Reference)
   PrintString!(execute=@n1.CastFailed, InString=B_BeanDog_CastFailed, bPrintToScreen, bPrintToLog, TextColor, Duration, Key) -> (then)
   Event Tick()
   n2 = Get BowTrajectory!(self:B Bow Athena Parent Object Reference=Get As B Bean Dog() -> (As B Bean Dog:B Bean Dog Object Reference)) -> (BowTrajectory:BP Projectile Trajectory Bow Generic Object Reference)
   > Is Valid(InputObject=@n2)
   > n3 = Cast To BP_ProjectileTrajectory_ExplosiveBow!(execute, Object=@n2) -> (then, CastFailed, AsBP Projectile Trajectory Explosive Bow:BP Projectile Trajectory Explosive Bow Object Reference) {is=title K2Node_DynamicCast}
   n4 = Get SplineComponent!(self:Fort Projectile Trajectory Object Reference=@n3) -> (SplineComponent:Spline Component Object Reference)
   > PrintString(InString="To String (Float)"(InDouble=GetSplineLength!(self:Spline Component Object Reference=@n4) -> (ReturnValue:float)))
  ]}]
variables [
 As B Bean Dog {type=B_BeanDog_C, default=null}
 As BP Projectile Trajectory Explosive Bow {type=BP_ProjectileTrajectory_ExplosiveBow_C, default=null}
]
widgets {
 info {
  inheritedWidgetCount=1
  parentClass=/Script/FortniteUI.FortWeaponReticleExtensionWidgetBase
  rootWidgetClass=&Image
  widgetCount=2
 }
 widgets [
  Reticle {
   widgetClassPath=&Image
   bIsVariable=true
   widget=Reticle
   properties {brush {
     bIsDynamicallyLoaded=false
     drawAs=Image
     imageType=FullColor
     mirroring=NoMirror
     resourceName=None
     resourceObject=/WildEstateWeapons/UI/BeanDog/Materials/M_UI_BeanDog_Reticle
     tiling=NoTile
     imageSize=<128 128>
     margin=<ltrb 0 0 0 0>
     outlineSettings {
      bUseBrushTransparency=false
      roundingType=HalfHeightRadius
      width=0
      color {colorUseRule=UseColor_Specified, specifiedColor=<rgba 0 0 0 0>}
      cornerRadii=<0 0 0 0>
     }
     tintColor {colorUseRule=UseColor_Specified, specifiedColor=<rgba 1 1 1 1>}
     uVRegion {bIsValid=false, max=<0 0>, min=<0 0>}
    }}
  }
  TutorialHighlight {widgetClassPath=&TutorialHighlight}
 ]
}
